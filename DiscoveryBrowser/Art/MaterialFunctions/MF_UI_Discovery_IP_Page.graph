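# UI material function for a Discovery IP page, compositing a header thumbnail over an input background color.
# Cursor position across the view is smoothstepped into a horizontal pixel offset, giving optional parallax.
# The fitted thumbnail is scaled about bottom-center, edge-clipped and faded toward the bottom with an eased mask before blending.
# Static switches toggle the header and a separate Opacity output from a rotatable linear UV gradient.
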MaterialFunction /DiscoveryBrowser/Art/MaterialFunctions/MF_UI_Discovery_IP_Page
graph {
 properties {previewBlendMode=BLEND_Translucent, previewMaterialDomain=MD_UI}
 nodes [
  n1 = Rotator(
   Coordinate=MaterialFunctionCall() {materialFunction=&GetUserInterfaceUV}.Normalized UV
   Time=ScalarParameter() {group=Fade Mask, parameterName=Fade Rotation}
  ) {speed=6.2832}
  FunctionOutput(None=StaticSwitchParameter(True=ComponentMask(None=@n1) {g=true}, False=Constant() {r=1}) {group=Fade Mask, parameterName=Is Fade Mask}) {outputName=Opacity, sortPriority=5}
  n2 = Saturate(None=FunctionInput() {inputName=Background Color})
  n3 = MaterialFunctionCall() {materialFunction=&GetUserInterfaceUV}
  n4 = Constant() {r=24}
  n5 = NamedRerouteDeclaration(None=Constant() {r=4}) {name=Parallax Offset Pixel Amount, nodeColor=<rgba 0 0 0 1>}
  n6 = NamedRerouteDeclaration(None=Constant2Vector()) {name=Widget Top Left on Screen, nodeColor=<rgba 0 0 0 1>}
  n7 = Add(A=@n6, B=ViewSize())
  n8 = NamedRerouteDeclaration(None=Clamp(
    None=MaterialFunctionCall() {materialFunction=/Engine/Functions/UserInterface/MF_UI_CursorPosition}.Result
    Min=@n6
    Max=NamedRerouteDeclaration(None=@n7) {name=Widget Bottom Right On Screen, nodeColor=<rgba 0 0 0 1>}
   )) {name=Pixel Range, nodeColor=<rgba 0 0 0 1>}
  n9 = NamedRerouteUsage()
  n10 = NamedRerouteUsage()
  n11 = SmoothStep(Value=MaterialFunctionCall(Input=ComponentMask(None=@n8) {r=true}, Input Low=ComponentMask(None=@n9) {r=true}, Input High=ComponentMask(None=@n10) {r=true}) {desc=Set Range for Mouse Position, materialFunction=&RemapValue_0-1}.Result)
  n12 = LinearInterpolate(
   A=Subtract(A=@n4, B=@n5)
   B=Add(A=@n4, B=@n5)
   Alpha=LinearInterpolate(
    B=NamedRerouteDeclaration(None=@n11) {name=Mouse Position Ratio, nodeColor=<rgba 0 0 0 1>}
    Alpha=ScalarParameter() {defaultValue=1, parameterName=UseParallax, sliderMax=1}
   )
  ) {desc=Padding Offset}
  n13 = Add(
   A=@n3.Normalized UV
   B=AppendVector(
    A=MaterialFunctionCall!(Value PX=@n12, UV Pixel Size, X or Y) {materialFunction=/FortUILibrary/MaterialFunctions/Basics/MF_UI_PixelsToNormalized}."Normalized (>1px)"
    B=Constant()
   )
  )
  n14 = TextureObjectParameter() {parameterName=Thumbnail, texture=/Game/UI/DiscoveryShared/Art/T_UI_Discovery_DefaultEmpty}
  n15 = Add(
   A=MaterialFunctionCall(UV=@n13, TextureSize=TextureProperty(Tex=@n14), ScalePivotPoint=Constant2Vector() {g=1, r=1}) {materialFunction=&MF_UI_UVScaling}.Scaled to Fit
   B=Constant2Vector() {r=0.01}
  )
  n16 = Constant2Vector() {g=1, r=0.5}
  n17 = Add(A=Multiply(A=Subtract(A=@n15, B=@n16), B=ScalarParameter() {defaultValue=1.1, parameterName=Mask Height}) {constB=1.1}, B=@n16)
  n18 = TextureSample(UVs=@n17, Tex=@n14) {samplerSource=SSM_Clamp_WorldGroupSettings}
  n19 = MaterialFunctionCall() {materialFunction=&GetUserInterfaceUV}
  n20 = MaterialFunctionCall(Input=OneMinus(None=ComponentMask(None=@n19.Normalized UV) {g=true}), Input Low=Constant(), Input High=Constant() {r=0.5}) {materialFunction=&RemapValue_0-1}
  n21 = Multiply(A=MaterialFunctionCall("Value (0-1)"=Saturate(None=@n20.Result)) {materialFunction=&MF_UI_EaseCurves}.EaseInSine) {constB=2}
  n22 = Saturate(None=Multiply(
    A=MaterialFunctionCall!(UVs=@n17, Alpha=@n18.A) {materialFunction=/FortUILibrary/MaterialFunctions/Basics/MF_UI_ClipEdgePixel}.Result
    B=Saturate(None=@n21)
   ))
  FunctionOutput(None=StaticSwitchParameter(True=LinearInterpolate(A=@n2, B=@n18.RGB, Alpha=@n22), False=@n2) {defaultValue=true, group=Header Settings, parameterName=Is Header}) {outputName=Color}
  LinearInterpolate(A=@n3.Normalized UV, B=@n13, Alpha=ScalarParameter() {defaultValue=1, parameterName=UseParallax, sliderMax=1})
 ]
}
