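# Translucent UI material for the animated Akita background in the Discovery browser, output through the shared IP Page function.
# Named reroutes build a dark tilted gradient, funnel-warped scrolling starfield, wave-warped shimmer mask, and linear and radial vignettes.
# Circle art packs three layers into RGB channels, each sampled with its own sine-driven sway in screen-scaled UVs.
# Masked lerps toward white composite the layers into the background color.
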Material /DiscoveryBrowser/Art/Materials/Backgrounds/M_UI_Discovery_Background_Akita
graph {
 properties {
  bAutomaticallySetUsageInEditor=false
  bUsedWithStaticMesh=false
  blendMode=BLEND_Translucent
  materialDomain=MD_UI
  usedShadingModels=Default Lit
 }
 nodes [
  n1 = NamedRerouteUsage()
  n2 = Constant() {r=1}
  n3 = Multiply(A=NamedRerouteUsage()) {constB=0.055}
  n4 = NamedRerouteUsage()
  MaterialFunctionCall!(Background Color=LinearInterpolate(A=LinearInterpolate(A=@n1, B=@n2, Alpha=@n3), B=Constant() {r=1}, Alpha=Multiply(A=@n4) {constB=0.85})) {materialFunction=/DiscoveryBrowser/Art/MaterialFunctions/MF_UI_Discovery_IP_Page}
  n5 = TextureCoordinate()
  n6 = AppendVector(A=ComponentMask(None=Add(A=@n5) {constB=-0.025}) {r=true}, B=ComponentMask(None=Add(A=@n5) {constB=0.15}) {g=true})
  n7 = Constant()
  n8 = Frac(None=Multiply(A=Time(), B=Constant() {r=0.03}) {constB=-0.25})
  n9 = MaterialFunctionCall!(
   UV=Multiply(A=MaterialFunctionCall!(ScreenUV=@n6, TextureSize, ScalePivotPoint) {materialFunction=/FortUILibrary/MaterialFunctions/Basics/MF_UI_ScreenSpaceUVScaling}.Scaled to Fit) {constB=0.75}
   Tiling=AppendVector(A=Constant() {r=6}, B=Constant() {r=1})
   Offset=OneMinus(None=AppendVector(A=@n7, B=@n8))
   Mask Size=Constant() {r=0.1}
   Blend Distance
  ) {materialFunction=/FortUILibrary/MaterialFunctions/Effects/MF_UI_Funnel}
  n10 = LinearInterpolate(
   A=TextureSample(UVs=@n9.Wrapped Funnel UV, Tex=TextureObject() {texture=/Game/Global/Textures/Default/PointSplatter/T_UI_Starfield}) {texture=&T_UI_Starfield}.RGB
   Alpha=OneMinus(None=@n9.Hole Mask)
  ) {constA=1, constB=0}
  n11 = NamedRerouteUsage()
  NamedRerouteDeclaration(None=Saturate(None=LinearInterpolate(B=@n10, Alpha=@n11) {constB=0})) {name=Stars Zoom, nodeColor=<rgba 0 0 0 1>}
  n12 = ScreenPosition()
  n13 = MaterialFunctionCall(UVs=@n12.ViewportUV, "Rotation Angle (0-1)"=Constant() {r=0.04}) {materialFunction=&CustomRotator}
  n14 = AppendVector(A=ComponentMask(None=Add(A=@n13.Rotated Values) {constB=0}) {g=true}, B=ComponentMask(None=Add(A=@n13.Rotated Values) {constB=0.3}) {g=true})
  n15 = Multiply(A=DotProduct(A=OneMinus(None=@n12.ViewportUV), B=@n12.ViewportUV)) {constB=2}
  n16 = LinearInterpolate(
   B=Power(Base=Saturate(None=@n15)) {constExponent=0.3}
   Alpha=MaterialFunctionCall("Value (0-1)"=@n14) {materialFunction=&MF_UI_EaseCurves}.EaseOutCubic
  ) {constA=1}
  n17 = Saturate(None=LinearInterpolate(B=OneMinus(None=@n14), Alpha=Saturate(None=@n16)))
  NamedRerouteDeclaration(None=ComponentMask(None=@n17) {g=true}) {name=Vignette B, nodeColor=<rgba 0 0 0 1>}
  NamedRerouteDeclaration(None=ComponentMask(None=@n17) {r=true}) {name=Vignette A, nodeColor=<rgba 0 0 0 1>}
  n18 = ComponentMask(None=MaterialFunctionCall(UVs=ScreenPosition().ViewportUV, "Rotation Angle (0-1)"=Constant() {r=0.04}) {materialFunction=&CustomRotator}.Rotated Values) {g=true}
  NamedRerouteDeclaration(None=LinearInterpolate(
    A=Constant3Vector() {constant=<rgba 0.0065 0.008 0.01 1>}
    B=Constant3Vector() {constant=<rgba 0.0325 0.0401 0.05 1>}
    Alpha=Add(A=@n18) {constB=-0.7}
   )) {name=Base Gradient, nodeColor=<rgba 0 0 0 1>}
  n19 = MaterialFunctionCall() {materialFunction=&MF_UI_ScreenSpaceUVScaling}
  n20 = MaterialFunctionCall!(
   UVs=Panner(Coordinate=Multiply(A=@n19.Scaled to Fit) {constB=16}) {speedX=0.03, speedY=0.01}
   Amplitude X=Constant() {r=0.3}
   Amplitude Y
   Frequency X=Constant() {r=0.3}
   Frequency Y=Constant() {r=0.5}
   Phase
   Motion=Constant() {r=0.1}
  ) {materialFunction=/FortUILibrary/MaterialFunctions/Effects/MF_UI_WaveWarp}
  n21 = TextureObject() {texture=/DiscoveryBrowser/Art/Textures/T_UI_HonMoon_Repeatable_BlurMask}
  NamedRerouteDeclaration(None=Saturate(None=TextureSample(UVs=@n20.Warped UVs, Tex=@n21) {samplerType=SAMPLERTYPE_Alpha}.R)) {name=ShimmerMask, nodeColor=<rgba 0 0 0 1>}
  n22 = Multiply(A=Time() {period=2}) {constB=0.175}
  n23 = Add(A=Multiply(A=Sine(None=@n22)) {constB=0.005}, B=ScreenPosition().ViewportUV)
  n24 = TextureProperty(Tex=NamedRerouteDeclaration(None=TextureObject() {samplerType=SAMPLERTYPE_LinearColor, texture=/DiscoveryBrowser/Art/Textures/T_UI_Akita_Background_Circles}) {name=Texture Object, nodeColor=<rgba 0 0 0 1>})
  n25 = Constant2Vector() {g=2}
  n26 = Add(
   A=Multiply(
    A=MaterialFunctionCall(ScreenUV=@n23, TextureSize=@n24, ScalePivotPoint=@n25) {materialFunction=&MF_UI_ScreenSpaceUVScaling}.Scaled to Fit
    B=ScalarParameter() {defaultValue=0.7, parameterName=Circles SIze}
   ) {constB=1.05}
   B=Constant2Vector() {g=0.05, r=0.02}
  )
  n27 = NamedRerouteUsage()
  n28 = Sine(None=Multiply(A=Time() {period=2}) {constB=0.1})
  n29 = TextureObject() {samplerType=SAMPLERTYPE_LinearColor, texture=&T_UI_Akita_Background_Circles}
  n30 = MaterialFunctionCall(
   ScreenUV=Add(A=Multiply(A=@n28) {constB=0.005}, B=ScreenPosition().ViewportUV)
   TextureSize=TextureProperty(Tex=NamedRerouteDeclaration(None=@n29) {name=Texture Object_2, nodeColor=<rgba 0 0 0 1>})
   ScalePivotPoint=Constant2Vector() {g=2}
  ) {materialFunction=&MF_UI_ScreenSpaceUVScaling}
  n31 = ScalarParameter() {defaultValue=0.7, parameterName=Circles SIze}
  n32 = TextureSample(UVs=Add(A=Multiply(A=@n30.Scaled to Fit, B=@n31) {constB=0.95}, B=Constant2Vector() {g=0.05, r=0.02}), Tex=NamedRerouteUsage()) {samplerType=SAMPLERTYPE_LinearColor, texture=&T_UI_Akita_Background_Circles}
  n33 = Sine(None=Multiply(A=Time() {period=2}) {constB=0.05})
  n34 = TextureObject() {samplerType=SAMPLERTYPE_LinearColor, texture=&T_UI_Akita_Background_Circles}
  n35 = MaterialFunctionCall(
   ScreenUV=Add(A=Multiply(A=@n33) {constB=0.005}, B=ScreenPosition().ViewportUV)
   TextureSize=TextureProperty(Tex=NamedRerouteDeclaration(None=@n34) {name=Texture Object_3, nodeColor=<rgba 0 0 0 1>})
   ScalePivotPoint=Constant2Vector() {g=2}
  ) {materialFunction=&MF_UI_ScreenSpaceUVScaling}
  n36 = ScalarParameter() {defaultValue=0.7, parameterName=Circles SIze}
  n37 = TextureSample(UVs=Add(A=Multiply(A=@n35.Scaled to Fit, B=@n36) {constB=0.95}, B=Constant2Vector() {g=0.05, r=0.02}), Tex=NamedRerouteUsage()) {samplerType=SAMPLERTYPE_LinearColor, texture=&T_UI_Akita_Background_Circles}
  n38 = Saturate(None=Add(
    A=TextureSample(UVs=@n26, Tex=@n27) {samplerType=SAMPLERTYPE_LinearColor, texture=&T_UI_Akita_Background_Circles}.R
    B=Add(A=@n32.G, B=@n37.B)
   ))
  n39 = NamedRerouteUsage()
  NamedRerouteDeclaration(None=LinearInterpolate(A=LinearInterpolate(B=@n38, Alpha=@n39) {constB=0}, Alpha=NamedRerouteUsage()) {constB=0}) {name=Circles, nodeColor=<rgba 0 0 0 1>}
  n40 = Multiply(A=ScreenPosition().ViewportUV) {constB=0.8}
  n41 = AppendVector(A=ComponentMask(None=Add(A=@n40) {constB=0}) {r=true}, B=ComponentMask(None=Add(A=@n40) {constB=0.2}) {g=true})
  n42 = Constant()
  n43 = Constant() {r=0.9}
  NamedRerouteDeclaration(None=Saturate(None=MaterialFunctionCall(UV=@n41, GradientStart=@n42, GradientEnd=@n43) {materialFunction=&MF_UI_RadialGradient}.Smooth)) {name=Vignette Circular, nodeColor=<rgba 0 0 0 1>}
 ]
}
