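# Translucent UI material for the creatures-themed Discovery browser background, feeding a layered orange color into MF_UI_Discovery_IP_Page.
# Screen-space scaled UVs drive radial vignette gradients, a rotated corner shape and a creature pattern grid.
# Two copies of the grid scroll horizontally in opposite directions, with complementary line masks built for alternating rows.
# A per-cell random value picks which channel of the packed creature texture each cell shows.
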Material /DiscoveryBrowser/Art/Materials/Backgrounds/M_UI_Discovery_Background_Creatures
graph {
 properties {
  bAutomaticallySetUsageInEditor=false
  bUsedWithStaticMesh=false
  blendMode=BLEND_Translucent
  materialDomain=MD_UI
  usedShadingModels=Default Lit
 }
 nodes [
  n1 = LinearInterpolate(
   A=LinearInterpolate(
    A=Constant3Vector() {constant=<rgba 1 0.4621 0.0284 1>}
    B=Constant3Vector() {constant=<rgba 0.8228 0.117 0.0021 1>}
    Alpha=NamedRerouteUsage()
   )
   B=LinearInterpolate(
    A=Constant3Vector() {constant=<rgba 0.8879 0.305 0.0152 1>}
    B=Constant3Vector() {constant=<rgba 0.8228 0.117 0.0021 1>}
    Alpha=NamedRerouteUsage()
   )
   Alpha=Multiply(A=NamedRerouteUsage()) {constB=0.8}
  )
  n2 = Constant3Vector() {constant=<rgba 0.57 0.0489 0.0015 1>}
  n3 = Multiply(A=NamedRerouteUsage())
  MaterialFunctionCall!(Background Color=LinearInterpolate(A=LinearInterpolate(A=@n1, B=@n2, Alpha=@n3), Alpha=NamedRerouteUsage())) {materialFunction=/DiscoveryBrowser/Art/MaterialFunctions/MF_UI_Discovery_IP_Page}
  n4 = MaterialFunctionCall!(ScreenUV=NamedRerouteDeclaration(None=ScreenPosition().ViewportUV) {name=Screen UV, nodeColor=<rgba 0 0 0 1>}, TextureSize, ScalePivotPoint) {materialFunction=/FortUILibrary/MaterialFunctions/Basics/MF_UI_ScreenSpaceUVScaling}
  NamedRerouteDeclaration(None=@n4.Scaled to Fill) {name=scale fill, nodeColor=<rgba 0 0 0 1>}
  NamedRerouteDeclaration(None=@n4.Scaled to Fit) {name=Scale Fit, nodeColor=<rgba 0 0 0 1>}
  n5 = NamedRerouteUsage()
  n6 = TextureSample(UVs=MaterialFunctionCall(UVs=NamedRerouteUsage(), Texture Scale=LinearInterpolate(Alpha=@n5) {constA=0.7, constB=0.5}) {materialFunction=&ScaleUVsByCenter}.UVs) {samplerType=SAMPLERTYPE_LinearColor, texture=/DiscoveryBrowser/Art/Textures/T_UI_Creatures_Pattern}
  n7 = NamedRerouteUsage()
  n8 = Add(A=Multiply(A=Time()) {constB=0.01}, B=@n7)
  n9 = Multiply(
   A=AppendVector(A=ComponentMask(None=@n8) {r=true}, B=ComponentMask(None=@n7) {g=true})
   B=ScalarParameter() {defaultValue=14, parameterName=Pattern Size}
  ) {constB=6}
  n10 = MaterialFunctionCall!(UV=Floor(None=@n9), Seed=Constant() {r=0.2}) {materialFunction=&MF_UI_Random}
  n11 = Saturate(None=Multiply(A=NamedRerouteDeclaration(None=@n10.Result) {name=Random Cells, nodeColor=<rgba 0 1 0.4375 1>}, B=Constant() {r=1}) {constB=1.3})
  n12 = Floor(None=@n11)
  n13 = Min(A=OneMinus(None=@n12), B=Round(None=@n11))
  n14 = LinearInterpolate(A=LinearInterpolate(A=LinearInterpolate(A=@n6.G, B=@n6.G, Alpha=@n13), B=@n6.R, Alpha=@n12), Alpha=NamedRerouteUsage()) {constA=1, constB=0}
  n15 = NamedRerouteUsage()
  n16 = TextureSample(UVs=MaterialFunctionCall(UVs=NamedRerouteUsage(), Texture Scale=LinearInterpolate(Alpha=@n15) {constA=0.7, constB=0.5}) {materialFunction=&ScaleUVsByCenter}.UVs) {samplerType=SAMPLERTYPE_LinearColor, texture=&T_UI_Creatures_Pattern}
  n17 = NamedRerouteUsage()
  n18 = Add(A=Multiply(A=Time()) {constB=-0.01}, B=@n17)
  n19 = Multiply(
   A=AppendVector(A=ComponentMask(None=@n18) {r=true}, B=ComponentMask(None=@n17) {g=true})
   B=ScalarParameter() {defaultValue=14, parameterName=Pattern Size}
  ) {constB=6}
  n20 = MaterialFunctionCall(UV=Floor(None=@n19), Seed=Constant() {r=0.2}) {materialFunction=&MF_UI_Random}
  n21 = Saturate(None=Multiply(A=NamedRerouteDeclaration(None=@n20.Result) {name=Random Cells_1, nodeColor=<rgba 0 1 0.4375 1>}) {constB=1.3})
  n22 = Floor(None=@n21)
  n23 = Min(A=OneMinus(None=@n22), B=Round(None=@n21))
  n24 = LinearInterpolate(A=LinearInterpolate(A=LinearInterpolate(A=@n16.G, B=@n16.G, Alpha=@n23), B=@n16.R, Alpha=@n22), Alpha=NamedRerouteUsage()) {constB=0}
  NamedRerouteDeclaration(None=Saturate(None=Add(A=@n14, B=@n24))) {name=Pattern, nodeColor=<rgba 0 0 0 1>}
  NamedRerouteDeclaration(None=Frac(None=@n9)) {name=Pattern UV, nodeColor=<rgba 0.5234 1 0 1>}
  n25 = MaterialFunctionCall(
   UVs=NamedRerouteUsage()
   Rotation Center=AppendVector(A=Constant() {r=0.82}, B=Constant() {r=0.5})
   "Rotation Angle (0-1)"=Constant() {r=0.63}
  ) {materialFunction=&CustomRotator}
  n26 = ComponentMask(None=@n25.Rotated Values) {r=true}
  n27 = SmoothStep(Value=@n26) {constMax=0.4975, constMin=0.5}
  n28 = MaterialFunctionCall(UVs=@n25.Rotated Values, "Rotation Angle (0-1)"=Constant() {r=-0.2}) {materialFunction=&CustomRotator}
  n29 = LinearInterpolate(
   B=SmoothStep(Value=@n26) {constMax=0.5175, constMin=0.52}
   Alpha=SmoothStep(Value=ComponentMask(None=@n28.Rotated Values) {g=true}) {constMax=0.5375, constMin=0.54}
  )
  NamedRerouteDeclaration(None=Saturate(None=Add(A=@n27, B=@n29))) {name=Corner Shape, nodeColor=<rgba 0 0 0 1>}
  n30 = Multiply(A=ScalarParameter() {defaultValue=14, parameterName=Pattern Size}, B=NamedRerouteUsage()) {constB=6}
  n31 = Frac(None=ComponentMask(None=Divide(A=@n30)) {g=true})
  NamedRerouteDeclaration(None=Ceil(None=Sine(None=@n31))) {name=Lines Mask A, nodeColor=<rgba 0 0 0 1>}
  n32 = Divide(A=Multiply(A=ScalarParameter() {defaultValue=14, parameterName=Pattern Size}, B=NamedRerouteUsage()) {constB=6})
  n33 = Sine(None=Frac(None=ComponentMask(None=@n32) {g=true}))
  NamedRerouteDeclaration(None=OneMinus(None=Ceil(None=@n33))) {name=Lines Mask B, nodeColor=<rgba 0 0 0 1>}
  n34 = ScalarParameter() {defaultValue=-0.2, group=Vignette, parameterName=Vig Mask X, sliderMax=1, sliderMin=-1}
  n35 = ScalarParameter() {defaultValue=0.03, group=Vignette, parameterName=Vig Max Y, sliderMax=1, sliderMin=-1}
  n36 = MaterialFunctionCall(
   UVs=Add(A=NamedRerouteUsage(), B=AppendVector(A=@n34, B=@n35))
   Texture Scale=AppendVector(
    A=ScalarParameter() {defaultValue=3, group=Vignette, parameterName=Vig Mask ScaleX, sliderMax=5}
    B=ScalarParameter() {defaultValue=1.5, group=Vignette, parameterName=Vig MaskAlpha ScaleY, sliderMax=5}
   )
  ) {materialFunction=&ScaleUVsByCenter}
  n37 = MaterialFunctionCall(UV=@n36.UVs, GradientStart=Constant() {r=1}, GradientEnd=Constant() {r=0.15}) {materialFunction=&MF_UI_RadialGradient}
  NamedRerouteDeclaration(None=Saturate(None=MaterialFunctionCall("Value (0-1)"=@n37.Smooth) {materialFunction=&MF_UI_EaseCurves}.EaseInOutSine)) {name=Radial Gradient Background, nodeColor=<rgba 0 0 0 1>}
  NamedRerouteDeclaration(None=Frac(None=@n19)) {name=Pattern UV_1, nodeColor=<rgba 0.5234 1 0 1>}
  n38 = MaterialFunctionCall(UV=@n36.UVs, GradientStart=Constant() {r=0.8}, GradientEnd=Constant() {r=0.1}) {materialFunction=&MF_UI_RadialGradient}
  NamedRerouteDeclaration(None=Saturate(None=MaterialFunctionCall("Value (0-1)"=@n38.Smooth) {materialFunction=&MF_UI_EaseCurves}.EaseInSine)) {name=Radial Gradient Pattern, nodeColor=<rgba 0 0 0 1>}
  n39 = NamedRerouteUsage()
  n40 = AppendVector(A=Constant() {r=-0.5}, B=Constant() {r=0.2})
  n41 = Constant() {r=3}
  n42 = Constant() {r=3}
  n43 = MaterialFunctionCall(
   UV=MaterialFunctionCall(UVs=Add(A=@n39, B=@n40), Texture Scale=AppendVector(A=@n41, B=@n42)) {materialFunction=&ScaleUVsByCenter}.UVs
   GradientStart=Constant() {r=0.6}
   GradientEnd=Constant() {r=-0.2}
  ) {materialFunction=&MF_UI_RadialGradient}
  NamedRerouteDeclaration(None=Saturate(None=MaterialFunctionCall("Value (0-1)"=@n43.Smooth) {materialFunction=&MF_UI_EaseCurves}.EaseInSine)) {name=Radial Gradient Overlay, nodeColor=<rgba 0 0 0 1>}
 ]
}
