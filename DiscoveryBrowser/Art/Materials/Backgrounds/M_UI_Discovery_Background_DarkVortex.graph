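# UI material for the Dark Vortex themed background in the Discovery browser, output through MF_UI_Discovery_IP_Page.
# Screen-space layers build a dark red gradient, grunge, and zombie shadows displaced by panning noise.
# Procedural halftone dots with randomized sizes and rotated scratches offset to either screen side form masks.
# The final color darkens to black by the maximum of two masks, with layers linked through named reroutes.
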Material /DiscoveryBrowser/Art/Materials/Backgrounds/M_UI_Discovery_Background_DarkVortex
graph {
 properties {
  bAutomaticallySetUsageInEditor=false
  bUsedWithStaticMesh=false
  blendMode=BLEND_Translucent
  materialDomain=MD_UI
  usedShadingModels=Default Lit
 }
 nodes [
  n1 = NamedRerouteUsage()
  n2 = Max(A=NamedRerouteUsage(), B=NamedRerouteUsage())
  MaterialFunctionCall!(Background Color=NamedRerouteDeclaration(None=LinearInterpolate(A=@n1, Alpha=@n2) {constB=0}) {name=Final Background, nodeColor=<rgba 0 0 0 1>}) {materialFunction=/DiscoveryBrowser/Art/MaterialFunctions/MF_UI_Discovery_IP_Page}
  n3 = NamedRerouteUsage()
  n4 = TextureObject() {samplerType=SAMPLERTYPE_Alpha, texture=/DiscoveryBrowser/Art/Textures/T_UI_DarkVortex_Background_Shadows}
  n5 = Add(
   A=MaterialFunctionCall!(ScreenUV, TextureSize=TextureProperty(Tex=@n4), ScalePivotPoint) {materialFunction=/FortUILibrary/MaterialFunctions/Basics/MF_UI_ScreenSpaceUVScaling}.ScaleX
   B=Constant2Vector() {g=0.4}
  )
  n6 = Constant2Vector() {g=1, r=1.5}
  n7 = Length(None=Subtract(A=MaterialFunctionCall(UVs=@n5, Texture Scale=@n6) {materialFunction=&ScaleUVsByCenter}.UVs) {constB=0.5}) {is=class MaterialExpressionLength}
  n8 = MaterialFunctionCall(Input=OneMinus(None=Saturate(None=@n7)), Input Low=Constant() {r=0.5}, Input High=Constant() {r=0.7}) {materialFunction=&RemapValue_0-1}
  n9 = NamedRerouteUsage()
  n10 = NamedRerouteDeclaration(None=TextureSample(
    UVs=Panner(Coordinate=@n9) {speedX=0.1}
    Tex=TextureObject() {samplerType=SAMPLERTYPE_Grayscale, texture=&LowResBlurredNoise}
   ) {samplerType=SAMPLERTYPE_Grayscale}.R) {name=Panning Clouds Mask, nodeColor=<rgba 0 0 0 1>}
  n11 = Constant() {desc=Zombie Displacement Intensity, r=0.02}
  n12 = TextureSample(UVs=Add(A=@n5, B=Multiply(A=@n10, B=@n11) {constB=0.01}), Tex=@n4) {samplerType=SAMPLERTYPE_Alpha}
  NamedRerouteDeclaration(None=LinearInterpolate(A=@n3, B=Multiply(A=@n3) {constB=0.4}, Alpha=Multiply(A=@n8.Result, B=@n12.R))) {name=Zombie Background, nodeColor=<rgba 0 0 0 1>}
  n13 = ScreenPosition()
  NamedRerouteDeclaration(None=MaterialFunctionCall(ScreenUV=NamedRerouteDeclaration(None=@n13.ViewportUV) {name=Screen UV, nodeColor=<rgba 0 0 0 1>}) {materialFunction=&MF_UI_ScreenSpaceUVScaling}.Scaled to Fit) {name=Scale Fit, nodeColor=<rgba 0 0 0 1>}
  n14 = Constant3Vector() {constant=<rgba 0.4072 0.0009 0.016 1>}
  n15 = Constant3Vector() {constant=<rgba 0.0595 0.0015 0.003 1>}
  n16 = NamedRerouteUsage()
  n17 = ComponentMask(None=@n16) {g=true}
  n18 = Constant()
  n19 = Constant() {r=0.8}
  n20 = SmoothStep(Value=Saturate(None=MaterialFunctionCall(Input=@n17, Input Low=@n18, Input High=@n19) {materialFunction=&RemapValue_0-1}.Result))
  n21 = Length(None=Subtract(A=@n16) {constB=0.5})
  NamedRerouteDeclaration(None=LinearInterpolate(A=LinearInterpolate(A=@n14, B=@n15, Alpha=@n20), B=Constant3Vector() {constant=<rgba 0.0482 0.0006 0.0021 1>}, Alpha=Saturate(None=@n21))) {name=Background Gradient, nodeColor=<rgba 0 0 0 1>}
  n22 = Rotator(Coordinate=NamedRerouteUsage(), Time=Constant() {r=-0.0156}) {speed=6.2832}
  n23 = Constant() {desc=Halftone Amount, r=80}
  n24 = Multiply(A=@n22, B=@n23) {constB=32}
  n25 = Constant() {desc=Halftone Min, r=0.7}
  n26 = Constant() {desc=Halftone Max, r=0.9}
  n27 = MaterialFunctionCall!(UV=Floor(None=@n24), Seed) {materialFunction=&MF_UI_Random}
  n28 = Saturate(None=Divide(A=NamedRerouteUsage(), B=Max(A=@n27.Result) {constB=0}))
  n29 = NamedRerouteUsage()
  n30 = Constant() {r=0.0156}
  n31 = ComponentMask(None=Rotator(Coordinate=NamedRerouteUsage(), Time=NamedRerouteDeclaration(None=@n30) {name=Halftone Rotation, nodeColor=<rgba 0 0 0 1>}) {speed=6.2832}) {r=true}
  n32 = Multiply(A=Min(A=@n31, B=OneMinus(None=@n31))) {constB=2}
  n33 = MaterialFunctionCall(
   Input=OneMinus(None=MaterialFunctionCall!("Value (0-1)"=@n32, Steps=@n23) {materialFunction=/FortUILibrary/MaterialFunctions/Basics/MF_UI_StepRatio}.Stepped Gradient)
   Input Low=Constant() {r=0.5}
   Input High=Constant() {r=1}
  ) {materialFunction=&RemapValue_0-1}
  n34 = Saturate(None=MaterialFunctionCall(A=Constant(), B=LinearInterpolate(Alpha=@n29) {constA=0.25}, C=@n33.Result, Alpha=@n33.Result) {materialFunction=&Lerp_3Color}.Result)
  n35 = Length(None=Subtract(A=@n22, B=Constant2Vector() {g=1, r=0.5}) {constB=0.5})
  n36 = Saturate(None=MaterialFunctionCall(Input=Saturate(None=@n35), Input Low=Constant(), Input High=Constant() {r=0.5}) {materialFunction=&RemapValue_0-1}.Result)
  n37 = MaterialFunctionCall(
   UVs=Frac(None=@n24)
   Size=Multiply(A=LinearInterpolate(A=@n25, B=@n26, Alpha=@n28) {constA=0.25, constB=0.75}, B=LinearInterpolate(B=@n34, Alpha=@n36) {constA=1})
  ) {materialFunction=&MF_UI_Circle}
  NamedRerouteDeclaration(None=LinearInterpolate(A=@n37.Fill, B=Multiply(A=@n37.Fill) {constB=0.125}, Alpha=OneMinus(None=@n36))) {name=Halftone Mask, nodeColor=<rgba 0 0 0 1>}
  n38 = NamedRerouteUsage()
  n39 = NamedRerouteUsage()
  n40 = TextureSample(
   UVs=MaterialFunctionCall(UVs=@n39, Texture Scale=Constant() {r=0.25}) {materialFunction=&ScaleUVsByCenter}.UVs
   Tex=TextureObject() {samplerType=SAMPLERTYPE_Alpha, texture=/DiscoveryBrowser/Art/Textures/T_UI_DarkVortex_Background_Grunge}
  ) {texture=&T_UI_DarkVortex_Background_Grunge}
  n41 = TextureSample(UVs=@n39, Tex=TextureObject() {samplerType=SAMPLERTYPE_Grayscale, texture=&LowResBlurredNoise}) {samplerType=SAMPLERTYPE_Grayscale}
  n42 = LinearInterpolate(Alpha=Multiply(A=@n41.R, B=@n41.R)) {constA=0.1}
  n43 = Saturate(None=Multiply(A=Multiply(A=@n40.R, B=@n42)) {constB=2})
  NamedRerouteDeclaration(None=Saturate(None=MaterialFunctionCall!(Base=@n38, Add=@n43) {materialFunction=/Engine/Functions/Engine_MaterialFunctions02/Math/MultiplyAdd}.Result)) {name=Grunge Background, nodeColor=<rgba 0 0 0 1>}
  n44 = NamedRerouteUsage()
  n45 = Constant2Vector() {r=-0.6}
  n46 = Constant2Vector() {r=0.6}
  n47 = Round(None=ComponentMask(None=@n44) {r=true})
  n48 = MaterialFunctionCall(ScreenUV=Subtract(A=@n44, B=LinearInterpolate(A=@n45, B=@n46, Alpha=@n47))) {materialFunction=&MF_UI_ScreenSpaceUVScaling}
  n49 = Subtract(A=NamedRerouteUsage()) {constB=0.2375}
  n50 = Subtract(
   A=MaterialFunctionCall(UVs=Rotator(Coordinate=@n48.ScaleY, Time=@n49) {speed=6.2832}, Texture Scale=Constant2Vector() {g=1, r=4}) {materialFunction=&ScaleUVsByCenter}.UVs
   B=Constant2Vector() {r=0.125}
  )
  n51 = TextureSample(UVs=@n50, Tex=TextureObject() {samplerType=SAMPLERTYPE_Alpha, texture=/DiscoveryBrowser/Art/Textures/T_UI_DarkVortex_Background_Scratches}) {samplerType=SAMPLERTYPE_Alpha}
  n52 = Multiply(A=@n51.R, B=@n51.R)
  n53 = Multiply(
   A=MaterialFunctionCall!(UVs=@n50, Alpha=@n51.A) {materialFunction=/FortUILibrary/MaterialFunctions/Basics/MF_UI_ClipEdgePixel}.Result
   B=Multiply(A=@n51.R, B=Multiply(A=@n51.R, B=@n52))
  )
  NamedRerouteDeclaration(None=Saturate(None=Multiply(A=@n53) {constB=2})) {name=Scratches Mask, nodeColor=<rgba 0 0 0 1>}
 ]
}
