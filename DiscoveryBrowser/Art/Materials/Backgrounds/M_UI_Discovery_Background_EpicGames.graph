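# Translucent UI material that supplies the Epic Games background for Discovery browser IP pages.
# A curve atlas gradient is sampled along screen-space UVs with adjustable rotation, offset and scale.
# An offset radial SmoothStep mask, built from aspect-scaled screen UVs, darkens that gradient away from its center.
# The result feeds the Background Color input of the shared MF_UI_Discovery_IP_Page function.
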Material /DiscoveryBrowser/Art/Materials/Backgrounds/M_UI_Discovery_Background_EpicGames
graph {
 properties {
  bAutomaticallySetUsageInEditor=false
  bUsedWithStaticMesh=false
  blendMode=BLEND_Translucent
  materialDomain=MD_UI
  usedShadingModels=Default Lit
 }
 nodes [
  n1 = NamedRerouteUsage()
  n2 = Add(
   A=MaterialFunctionCall!(ScreenUV=NamedRerouteUsage(), TextureSize, ScalePivotPoint=Constant2Vector() {g=1, r=1}) {materialFunction=/FortUILibrary/MaterialFunctions/Basics/MF_UI_ScreenSpaceUVScaling}.ScaleY
   B=Constant2Vector() {g=0.1, r=0.1}
  )
  n3 = Add(A=Multiply(A=Subtract(A=@n2) {constB=0.5}) {constB=0.375}) {constB=0.5}
  n4 = Saturate(None=Length(None=ConstantBiasScale(None=@n3) {bias=-0.5, scale=2}) {is=class MaterialExpressionLength})
  MaterialFunctionCall!(Background Color=LinearInterpolate(A=@n1, B=Multiply(A=@n1) {constB=0.375}, Alpha=SmoothStep(Value=@n4))) {materialFunction=/DiscoveryBrowser/Art/MaterialFunctions/MF_UI_Discovery_IP_Page}
  n5 = NamedRerouteUsage()
  n6 = ScalarParameter() {parameterName=Base Gradient Rotation}
  n7 = ScalarParameter() {parameterName=Base Gradient Offset}
  n8 = Constant()
  n9 = Multiply(
   A=Add(A=Rotator(Coordinate=@n5, Time=@n6) {speed=6.2832}, B=AppendVector(A=@n7, B=@n8))
   B=ScalarParameter() {defaultValue=1, parameterName=Base Gradient Scale}
  )
  n10 = CurveAtlasRowParameter(CurveTime=Saturate(None=ComponentMask(None=@n9) {r=true})) {
   atlas=/DiscoveryBrowser/Art/Textures/CA_Discovery_Background_Epic
   curve=/DiscoveryBrowser/Art/Textures/Curve_Discovery_Background_Epic
   parameterName=BackgroundCurveGradient
  }
  NamedRerouteDeclaration(None=@n10) {name=Base Gradient, nodeColor=<rgba 0 0 0 1>}
  n11 = ScreenPosition()
  NamedRerouteDeclaration(None=MaterialFunctionCall(ScreenUV=NamedRerouteDeclaration(None=@n11.ViewportUV) {name=UV, nodeColor=<rgba 0 0 0 1>}) {materialFunction=&MF_UI_ScreenSpaceUVScaling}.ScaleY) {name=Scale Y UV, nodeColor=<rgba 0 0 0 1>}
 ]
}
