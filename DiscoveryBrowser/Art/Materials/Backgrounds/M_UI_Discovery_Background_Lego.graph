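# Translucent UI material for the Lego background in the Discovery browser, passed through the shared IP page material function.
# A purple screen-space gradient combines a vertical blend with an aspect-scaled radial falloff.
# A slowly panning, fixed-rotation Lego texture has its red and green channels smoothstepped and summed into a pattern mask.
# Overlay blending applies detail to the gradient, faded by screen height and a second radial mask.
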Material /DiscoveryBrowser/Art/Materials/Backgrounds/M_UI_Discovery_Background_Lego
graph {
 properties {
  bAutomaticallySetUsageInEditor=false
  bUsedWithStaticMesh=false
  blendMode=BLEND_Translucent
  materialDomain=MD_UI
  usedShadingModels=Default Lit
 }
 nodes [
  MaterialFunctionCall!(Background Color=NamedRerouteUsage()) {materialFunction=/DiscoveryBrowser/Art/MaterialFunctions/MF_UI_Discovery_IP_Page}
  n1 = Constant3Vector() {constant=<rgba 0.3372 0.0382 0.4397 1>}
  n2 = Constant3Vector() {constant=<rgba 0.0908 0.0091 0.1195 1>}
  n3 = ScreenPosition()
  n4 = ComponentMask(None=@n3.ViewportUV) {g=true}
  n5 = Constant2Vector() {g=0.5, r=0.5}
  n6 = MaterialFunctionCall(
   UVs=MaterialFunctionCall!(ScreenUV=Add(A=@n3.ViewportUV, B=@n5), TextureSize, ScalePivotPoint) {materialFunction=/FortUILibrary/MaterialFunctions/Basics/MF_UI_ScreenSpaceUVScaling}.ScaleY
   Texture Scale=Constant() {r=6}
  ) {materialFunction=&ScaleUVsByCenter}
  n7 = NamedRerouteDeclaration(None=LinearInterpolate(
    A=LinearInterpolate(A=@n1, B=@n2, Alpha=@n4)
    B=Constant3Vector() {constant=<rgba 0.0561 0.0021 0.0782 1>}
    Alpha=MaterialFunctionCall(UV=@n6.UVs) {materialFunction=&MF_UI_RadialGradient}.Smooth
   )) {name=Gradient, nodeColor=<rgba 0 0 0 1>}
  n8 = NamedRerouteUsage()
  n9 = MaterialFunctionCall(Base=@n7, Blend=@n8) {materialFunction=&Blend_Overlay}
  n10 = Multiply(A=@n8) {constB=0.5}
  n11 = ScreenPosition()
  n12 = ComponentMask(None=@n11.ViewportUV) {g=true}
  n13 = Constant() {r=3}
  n14 = OneMinus(None=MaterialFunctionCall(UV=MaterialFunctionCall(UVs=@n11.ViewportUV, Texture Scale=@n13) {materialFunction=&ScaleUVsByCenter}.UVs) {materialFunction=&MF_UI_RadialGradient}.Smooth)
  NamedRerouteDeclaration(None=LinearInterpolate(A=LinearInterpolate(A=@n7, B=@n9.Result, Alpha=@n10), B=NamedRerouteUsage(), Alpha=LinearInterpolate(A=@n12, Alpha=@n14))) {name=Background, nodeColor=<rgba 0 0 0 1>}
  n15 = Constant() {r=0.425}
  n16 = Constant() {r=0.05}
  n17 = Multiply(A=ScalarParameter() {defaultValue=2, group=Topo Lines, parameterName=Topo Panner Speed X}) {constB=0.001}
  n18 = Multiply(A=ScalarParameter() {defaultValue=1.5, group=Topo Lines, parameterName=Topo Panner Speed y}) {constB=0.001}
  n19 = MaterialFunctionCall(
   UVs=Panner(Coordinate=ScreenPosition().ViewportUV, Speed=AppendVector(A=@n17, B=@n18)) {speedX=2, speedY=2}
   Texture Scale=Constant() {r=0.75}
  ) {materialFunction=&ScaleUVsByCenter}
  n20 = TextureSample(UVs=Rotator(
    Coordinate=MaterialFunctionCall(ScreenUV=@n19.UVs) {materialFunction=&MF_UI_ScreenSpaceUVScaling}.Scaled to Fit
    Time=Constant() {r=0.5}
   ) {speed=6.2832}) {texture=/DiscoveryBrowser/Art/Textures/T_UI_Discovery_Lego_Pattern}
  n21 = Multiply(A=Constant() {r=0.75}, B=SmoothStep(Min=@n15, Max=Add(A=@n15, B=@n16) {constB=0.21}, Value=@n20.R)) {constB=0.75}
  n22 = Constant() {r=0.45}
  n23 = SmoothStep(Min=@n22, Max=Add(A=@n22, B=Constant() {r=0.05}) {constB=0.21}, Value=LinearInterpolate(A=@n20.G, B=@n20.G, Alpha=Constant()))
  NamedRerouteDeclaration(None=Saturate(None=Add(A=@n21, B=@n23))) {name=Pattern, nodeColor=<rgba 0 0 0 1>}
 ]
}
