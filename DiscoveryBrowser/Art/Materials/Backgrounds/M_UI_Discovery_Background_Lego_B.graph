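# Translucent UI material for a Lego-themed Discovery browser background, feeding its color into MF_UI_Discovery_IP_Page.
# A screen-space vertical gradient between two purples is darkened slightly toward black by a mask.
# A halftone mask builds circles on a rotated, slowly scrolling grid, sized by a stepped vertical edge falloff and cut by an offset radial falloff.
# Per-cell random sizing is wired in, but equal min and max values neutralize it.
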Material /DiscoveryBrowser/Art/Materials/Backgrounds/M_UI_Discovery_Background_Lego_B
graph {
 properties {
  bAutomaticallySetUsageInEditor=false
  bUsedWithStaticMesh=false
  blendMode=BLEND_Translucent
  materialDomain=MD_UI
  usedShadingModels=Default Lit
 }
 nodes [
  MaterialFunctionCall!(Background Color=NamedRerouteUsage()) {materialFunction=/DiscoveryBrowser/Art/MaterialFunctions/MF_UI_Discovery_IP_Page}
  n1 = Constant3Vector() {constant=<rgba 0.2384 0.0037 0.5583 1>}
  n2 = Constant3Vector() {constant=<rgba 0.1616 0.0048 0.3802 1>}
  n3 = ComponentMask(None=ScreenPosition().ViewportUV) {g=true}
  n4 = NamedRerouteUsage()
  NamedRerouteDeclaration(None=LinearInterpolate(A=LinearInterpolate(A=@n1, B=@n2, Alpha=@n3), B=Constant(), Alpha=Multiply(A=@n4) {constB=0.25})) {name=Background, nodeColor=<rgba 0 0 0 1>}
  n5 = MaterialFunctionCall!(ScreenUV=NamedRerouteDeclaration(None=ScreenPosition().ViewportUV) {name=Screen UV, nodeColor=<rgba 0 0 0 1>}, TextureSize, ScalePivotPoint) {materialFunction=/FortUILibrary/MaterialFunctions/Basics/MF_UI_ScreenSpaceUVScaling}
  NamedRerouteDeclaration(None=@n5.Scaled to Fill) {name=scale fill, nodeColor=<rgba 0 0 0 1>}
  NamedRerouteDeclaration(None=@n5.Scaled to Fit) {name=Scale Fit, nodeColor=<rgba 0 0 0 1>}
  n6 = NamedRerouteUsage()
  n7 = OneMinus(None=Rotator(Coordinate=@n6, Time=Constant() {r=0.5}))
  n8 = Time() {period=1}
  n9 = Add(A=@n7, B=Multiply(A=@n8) {constB=0.0015})
  n10 = Add(A=@n7, B=Multiply(A=@n8) {constB=-0.0035})
  n11 = Constant() {desc=Halftone Amount, r=180}
  n12 = Multiply(A=AppendVector(A=ComponentMask(None=@n9) {r=true}, B=ComponentMask(None=@n10) {g=true}), B=@n11) {constB=32}
  n13 = Constant() {desc=Halftone Min, r=0.35}
  n14 = Constant() {desc=Halftone Max, r=0.35}
  n15 = Floor(None=@n12)
  n16 = Saturate(None=Max(A=MaterialFunctionCall!(UV=@n15, Seed) {materialFunction=&MF_UI_Random}.Result) {constB=0})
  n17 = ComponentMask(None=NamedRerouteUsage()) {g=true}
  n18 = Multiply(A=Min(A=@n17, B=OneMinus(None=@n17))) {constB=2}
  n19 = Saturate(None=MaterialFunctionCall(
    Input=MaterialFunctionCall!("Value (0-1)"=@n18, Steps=@n11) {materialFunction=/FortUILibrary/MaterialFunctions/Basics/MF_UI_StepRatio}.Stepped Gradient
    Input Low=Constant()
    Input High=Constant() {r=0.25}
   ) {materialFunction=&RemapValue_0-1}.Result)
  n20 = Length(None=Subtract(A=OneMinus(None=@n6), B=Constant2Vector() {g=1, r=0.25}) {constB=0.5}) {is=class MaterialExpressionLength}
  n21 = Saturate(None=MaterialFunctionCall(Input=Saturate(None=@n20), Input Low=Constant() {r=0.4}, Input High=Constant() {r=0.5}) {materialFunction=&RemapValue_0-1}.Result)
  n22 = MaterialFunctionCall(
   UVs=Frac(None=@n12)
   Size=Multiply(A=LinearInterpolate(A=@n13, B=@n14, Alpha=@n16) {constA=0.25, constB=0.75}, B=LinearInterpolate(B=@n19, Alpha=@n21) {constB=0})
  ) {materialFunction=&MF_UI_Circle}
  n23 = OneMinus(None=@n21)
  NamedRerouteDeclaration(None=Saturate(None=LinearInterpolate(A=@n22.Fill, Alpha=@n23) {constB=0})) {name=Halftone Mask Small, nodeColor=<rgba 0 0 0 1>}
 ]
}
