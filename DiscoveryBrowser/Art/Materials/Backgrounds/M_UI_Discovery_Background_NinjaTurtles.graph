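# Translucent UI material that draws the Ninja Turtles themed background for Discovery browser IP pages through MF_UI_Discovery_IP_Page.
# Screen-space vertical and radial blends build a dark purple gradient, which receives an overlay blend and slowly panning blurred-noise modulation.
# A divide-based brightening pass and a vertical/radial masked blend finish the Background color.
# Named reroutes link the stages, including a Ninja Turtles pattern texture scaled to fit the screen.
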Material /DiscoveryBrowser/Art/Materials/Backgrounds/M_UI_Discovery_Background_NinjaTurtles
graph {
 properties {
  bAutomaticallySetUsageInEditor=false
  bUsedWithStaticMesh=false
  blendMode=BLEND_Translucent
  materialDomain=MD_UI
  usedShadingModels=Default Lit
 }
 nodes [
  MaterialFunctionCall!(Background Color=NamedRerouteUsage()) {materialFunction=/DiscoveryBrowser/Art/MaterialFunctions/MF_UI_Discovery_IP_Page}
  n1 = TextureSample(UVs=Panner(Coordinate=ScreenPosition().ViewportUV) {speedX=-0.05, speedY=0.025}) {samplerType=SAMPLERTYPE_Grayscale, texture=&LowResBlurredNoise}
  n2 = Constant3Vector() {constant=<rgba 0.126 0 0.4427 1>}
  n3 = Constant3Vector() {constant=<rgba 0.0467 0 0.1384 1>}
  n4 = ScreenPosition()
  n5 = ComponentMask(None=@n4.ViewportUV) {g=true}
  n6 = Constant2Vector() {g=0.5, r=0.5}
  n7 = MaterialFunctionCall(
   UVs=MaterialFunctionCall!(ScreenUV=Add(A=@n4.ViewportUV, B=@n6), TextureSize, ScalePivotPoint) {materialFunction=/FortUILibrary/MaterialFunctions/Basics/MF_UI_ScreenSpaceUVScaling}.ScaleY
   Texture Scale=Constant() {r=6}
  ) {materialFunction=&ScaleUVsByCenter}
  n8 = NamedRerouteDeclaration(None=LinearInterpolate(
    A=LinearInterpolate(A=@n2, B=@n3, Alpha=@n5)
    B=Constant3Vector() {constant=<rgba 0.0081 0 0.0365 1>}
    Alpha=MaterialFunctionCall(UV=@n7.UVs) {materialFunction=&MF_UI_RadialGradient}.Smooth
   )) {name=Gradient, nodeColor=<rgba 0 0 0 1>}
  n9 = NamedRerouteUsage()
  n10 = LinearInterpolate(
   A=@n8
   B=MaterialFunctionCall(Base=@n8, Blend=@n9) {materialFunction=&Blend_Overlay}.Result
   Alpha=Saturate(None=Multiply(A=@n9) {constB=24})
  ) {constAlpha=1}
  n11 = LinearInterpolate(A=Multiply(A=@n1.R, B=@n10), B=@n10, Alpha=@n1.R)
  n12 = NamedRerouteUsage()
  n13 = Divide(A=@n11, B=Max(A=Multiply(A=@n12) {constB=0.1}) {constB=0})
  n14 = NamedRerouteUsage()
  n15 = Multiply(A=@n14, B=@n14)
  n16 = ScreenPosition()
  n17 = ComponentMask(None=@n16.ViewportUV) {g=true}
  n18 = Constant() {r=1.5}
  n19 = OneMinus(None=MaterialFunctionCall(UV=MaterialFunctionCall(UVs=@n16.ViewportUV, Texture Scale=@n18) {materialFunction=&ScaleUVsByCenter}.UVs) {materialFunction=&MF_UI_RadialGradient}.Smooth)
  NamedRerouteDeclaration(None=LinearInterpolate(A=LinearInterpolate(A=@n11, B=@n13, Alpha=@n15) {constB=0.1}, B=NamedRerouteUsage(), Alpha=LinearInterpolate(A=@n17, Alpha=@n19))) {name=Background, nodeColor=<rgba 0 0 0 1>}
  n20 = MaterialFunctionCall(UVs=ScreenPosition().ViewportUV, Texture Scale=Constant() {r=0.4}) {materialFunction=&ScaleUVsByCenter}
  NamedRerouteDeclaration(None=TextureSample(UVs=MaterialFunctionCall(ScreenUV=@n20.UVs) {materialFunction=&MF_UI_ScreenSpaceUVScaling}.Scaled to Fit) {samplerType=SAMPLERTYPE_Alpha, texture=/DiscoveryBrowser/Art/Textures/T_UI_Discovery_NinjaTurtles_Pattern}.R) {name=Pattern, nodeColor=<rgba 0 0 0 1>}
 ]
}
