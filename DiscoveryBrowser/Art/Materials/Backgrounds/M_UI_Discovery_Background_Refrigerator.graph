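# Translucent UI material that draws the animated "Honmoon" background for the Refrigerator Discovery page, fed into MF_UI_Discovery_IP_Page.
# Procedural line fields come from UVs that are offset, wave-warped, rotated and tiled with Frac, alongside blur-texture distortion maps.
# Radial glows, a vignette, panning shimmer, a flare texture and a circle-SDF ring supply further masks, using screen-space UV scaling.
# Named reroutes carry masks and colours into chained lerps and blend modes.
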Material /DiscoveryBrowser/Art/Materials/Backgrounds/M_UI_Discovery_Background_Refrigerator
graph {
 properties {
  bAutomaticallySetUsageInEditor=false
  bUsedWithStaticMesh=false
  blendMode=BLEND_Translucent
  materialDomain=MD_UI
  usedShadingModels=Default Lit
 }
 nodes [
  Constant() {r=1.2}
  Constant() {r=-0.09}
  Constant() {r=1}
  Constant() {r=1}
  Constant() {r=-1.66}
  Constant()
  n1 = NamedRerouteUsage()
  MaterialFunctionCall!(Input UVs=@n1, Translate=AppendVector(A=Constant(), B=Constant())) {materialFunction=/FortUILibrary/MaterialFunctions/Basics/MF_UI_Translate}
  n2 = Constant3Vector() {constant=<rgba 0.8878 0.0099 1 1>}
  n3 = Constant3Vector() {constant=<rgba 0 0.0482 1 1>}
  n4 = MaterialFunctionCall() {materialFunction=&LinearGradient}
  n5 = LinearInterpolate(
   A=LinearInterpolate(A=Constant3Vector() {constant=<rgba 0.0041 0.0055 0.0469 1>}, B=NamedRerouteUsage(), Alpha=NamedRerouteUsage()) {constAlpha=1}
   B=NamedRerouteDeclaration(None=LinearInterpolate(A=@n2, B=@n3, Alpha=@n4.UGradient)) {name=Honmoon colour, nodeColor=<rgba 0 0 0 1>}
   Alpha=NamedRerouteUsage()
  )
  n6 = MaterialFunctionCall!(Base=@n5, Blend=Constant3Vector() {constant=<rgba 0.0008 0.3728 1 1>}) {materialFunction=/Engine/Functions/Engine_MaterialFunctions03/Blends/Blend_LinearLight}
  n7 = NamedRerouteUsage()
  n8 = Constant3Vector() {constant=<rgba 0 0.3025 1 1>}
  n9 = LinearInterpolate(
   A=LinearInterpolate(
    A=LinearInterpolate(A=@n5, B=@n6.Result, Alpha=@n7)
    B=NamedRerouteDeclaration(None=@n8) {name=Flash colour, nodeColor=<rgba 0 0 0 1>}
    Alpha=NamedRerouteUsage()
   )
   B=NamedRerouteUsage()
   Alpha=NamedRerouteUsage()
  )
  n10 = NamedRerouteUsage()
  n11 = NamedRerouteUsage()
  n12 = Constant3Vector() {constant=<rgba 0 0.2431 1 1>}
  n13 = Multiply(A=NamedRerouteUsage()) {constB=8}
  MaterialFunctionCall!(Background Color=LinearInterpolate(A=LinearInterpolate(A=@n9, B=@n10, Alpha=@n11), B=LinearInterpolate(B=@n12, Alpha=@n13), Alpha=NamedRerouteUsage())) {materialFunction=/DiscoveryBrowser/Art/MaterialFunctions/MF_UI_Discovery_IP_Page}
  n14 = Constant()
  n15 = Constant() {r=-0.05}
  n16 = MaterialFunctionCall(
   UVs=MaterialFunctionCall(Input UVs=NamedRerouteUsage(), Translate=AppendVector(A=@n14, B=@n15)) {materialFunction=&MF_UI_Translate}.UVs raw
   Texture Scale=AppendVector(A=Constant() {r=1.8}, B=Constant() {r=1.8})
  ) {materialFunction=&ScaleUVsByCenter}
  n17 = Add(A=@n16.UVs, B=Multiply(A=@n16.UVs, B=NamedRerouteUsage()) {constB=2})
  n18 = Constant() {r=0.002}
  n19 = Constant() {r=2}
  n20 = Constant()
  n21 = MaterialFunctionCall!(FloatValue, Speed=ScalarParameter() {defaultValue=0.2, group=Distortion, parameterName=Distortion_MotionSpeed}) {materialFunction=/FortUILibrary/MaterialFunctions/Basics/MF_UI_LinearTime}
  n22 = Multiply(
   A=MaterialFunctionCall(
    UVs=MaterialFunctionCall!(UVs=@n17, Amplitude X=@n18, Amplitude Y=@n18, Frequency X=@n19, Frequency Y=@n19, Phase=@n20, Motion=@n21.0-1 Loop) {materialFunction=/FortUILibrary/MaterialFunctions/Effects/MF_UI_WaveWarp}.Warped UVs
    "Rotation Angle (0-1)"=Constant() {r=0.78}
   ) {materialFunction=&CustomRotator}.Rotated Values
   B=Constant2Vector() {g=1, r=80}
  )
  n23 = NamedRerouteDeclaration(None=MaterialFunctionCall!(UVs=Frac(None=@n22), PointA, PointB, Thickness, OutlineThickness, OutlineAlignment, GlowMin, GlowMax) {materialFunction=/FortUILibrary/MaterialFunctions/Shapes/MF_UI_LineSegment}.Fill) {name=LinesMask, nodeColor=<rgba 0 0 0 1>}
  NamedRerouteDeclaration(None=Multiply(A=@n23, B=Constant() {r=0.2})) {name=Honmoon Mask_dim, nodeColor=<rgba 0 0 0 1>}
  n24 = Subtract(A=NamedRerouteUsage(), B=Constant() {r=0.5}) {constB=0.1}
  NamedRerouteDeclaration(None=Saturate(None=Length(None=@n24) {is=class MaterialExpressionLength})) {name=Background Mask, nodeColor=<rgba 0 0 0 1>}
  n25 = LinearInterpolate(B=Constant3Vector() {constant=<rgba 0.3409 0.1094 1 1>}, Alpha=NamedRerouteUsage())
  n26 = Constant3Vector() {constant=<rgba 0.3409 0.1094 1 1>}
  n27 = NamedRerouteUsage()
  NamedRerouteDeclaration(None=LinearInterpolate(A=LinearInterpolate(A=@n25, B=@n26, Alpha=@n27), B=Constant3Vector() {constant=<rgba 0.0576 0.5292 0.776 1>}, Alpha=NamedRerouteUsage())) {name=glow colours, nodeColor=<rgba 0 0 0 1>}
  n28 = LinearInterpolate(
   A=LinearInterpolate(
    A=Constant3Vector() {constant=<rgba 0 0.0158 0.0938 1>}
    B=Constant3Vector() {constant=<rgba 0.0032 0.0043 0.0365 1>}
    Alpha=NamedRerouteUsage()
   )
   B=NamedRerouteUsage()
   Alpha=NamedRerouteUsage()
  )
  n29 = Constant3Vector() {constant=<rgba 0.7083 0.7083 0.7083 0>}
  NamedRerouteDeclaration(None=LinearInterpolate(
    A=@n28
    B=MaterialFunctionCall!(Blend=@n29, Base=@n28) {materialFunction=/Engine/Functions/Engine_MaterialFunctions03/Blends/Blend_ColorDodge}.Result
    Alpha=NamedRerouteUsage()
   )) {name=Background, nodeColor=<rgba 0 0 0 1>}
  n30 = MaterialFunctionCall(
   Input UVs=MaterialFunctionCall(UVs=NamedRerouteUsage(), Texture Scale=ScalarParameter() {defaultValue=1.8, group=Distortion, parameterName=Distortion_MapScale}) {materialFunction=&ScaleUVsByCenter}.UVs
   Translate=AppendVector(A=Constant() {r=-1.0764}, B=Constant())
  ) {materialFunction=&MF_UI_Translate}
  n31 = TextureObject() {texture=/DiscoveryBrowser/Art/Textures/T_UI_HonMoon_Repeatable_BlurMask}
  n32 = Constant() {r=1}
  n33 = NamedRerouteDeclaration(None=SmoothStep(Value=TextureSample!(UVs=@n30.UVs raw, Tex=@n31, Level=@n32, Apply View MipBias) {mipValueMode=TMVM_MipLevel}.R)) {name=DistortionTexture, nodeColor=<rgba 0 0 0 1>}
  NamedRerouteDeclaration(None=Multiply(A=Multiply(A=@n33, B=@n33), B=ScalarParameter() {defaultValue=1.2, group=Distortion, parameterName=Distortion_Strength})) {name=DistortionMap, nodeColor=<rgba 0 0 0 1>}
  n34 = MaterialFunctionCall(UVs=@n1, Texture Scale=AppendVector(A=Constant() {r=3.2}, B=Constant() {r=0.95})) {materialFunction=&ScaleUVsByCenter}
  n35 = Constant() {r=0.5}
  NamedRerouteDeclaration(None=SmoothStep(
    Max=Constant() {r=0.5}
    Value=MaterialFunctionCall!(UVs=@n34.UVs, Radius=@n35) {materialFunction=/FortUILibrary/MaterialFunctions/Shapes/SDFs/MF_SDF_Circle}.SDF
   ) {constMax=0.5}) {name=VignetteMask, nodeColor=<rgba 0 0 0 1>}
  n36 = Panner(Coordinate=ComponentMask(None=NamedRerouteUsage()) {r=true}) {speedX=-0.04, speedY=0.226}
  n37 = Constant() {r=0.05}
  n38 = Constant() {r=0.5}
  n39 = Constant() {r=0.9}
  n40 = MaterialFunctionCall(Speed=ScalarParameter() {defaultValue=0.2, group=Distortion, parameterName=Distortion_MotionSpeed}) {materialFunction=&MF_UI_LinearTime}
  n41 = Multiply(A=TextureSample!(
    UVs=MaterialFunctionCall(UVs=@n36, Amplitude X=@n37, Frequency X=@n38, Frequency Y=@n39, Motion=@n40.0-1 Loop) {materialFunction=&MF_UI_WaveWarp}.Warped UVs
    Tex=TextureObject() {texture=&T_UI_HonMoon_Repeatable_BlurMask}
    Apply View MipBias
   ) {samplerType=SAMPLERTYPE_Alpha}.R) {constB=2}
  n42 = NamedRerouteUsage()
  n43 = NamedRerouteUsage()
  n44 = Subtract(A=ComponentMask(None=@n43) {r=true}) {constB=0.5}
  n45 = Subtract(A=OneMinus(None=ComponentMask(None=@n43) {r=true})) {constB=0.15}
  n46 = Multiply(A=Multiply(A=Saturate(None=@n44), B=Saturate(None=@n45))) {constB=20}
  NamedRerouteDeclaration(None=LinearInterpolate(B=NamedRerouteDeclaration(None=@n41) {name=ShimmerMask, nodeColor=<rgba 0 0 0 1>}, Alpha=LinearInterpolate(B=@n42, Alpha=@n46))) {name=ShimmerVinetteMask, nodeColor=<rgba 0 0 0 1>}
  n47 = MaterialFunctionCall!(ScreenUV=NamedRerouteDeclaration(None=ScreenPosition().ViewportUV) {name=Screen UV, nodeColor=<rgba 0 0 0 1>}, TextureSize, ScalePivotPoint) {materialFunction=/FortUILibrary/MaterialFunctions/Basics/MF_UI_ScreenSpaceUVScaling}
  NamedRerouteDeclaration(None=@n47.Scaled to Fit) {name=Scale Fit, nodeColor=<rgba 0 0 0 1>}
  NamedRerouteDeclaration(None=Multiply(A=NamedRerouteUsage(), B=NamedRerouteUsage())) {name=ShimmerLinesMask, nodeColor=<rgba 0 0 0 1>}
  NamedRerouteDeclaration(None=@n47.ScaleX) {name=Scale X, nodeColor=<rgba 0 0 0 1>}
  n48 = MaterialFunctionCall(
   Input UVs=MaterialFunctionCall(UVs=NamedRerouteUsage(), Texture Scale=Constant() {r=0.9}) {materialFunction=&ScaleUVsByCenter}.UVs
   Translate=AppendVector(A=Constant() {r=0.3}, B=Constant() {r=1.02})
  ) {materialFunction=&MF_UI_Translate}
  n49 = TextureObject() {texture=&T_UI_HonMoon_Repeatable_BlurMask}
  n50 = Constant() {r=1}
  n51 = NamedRerouteDeclaration(None=SmoothStep(Value=TextureSample!(UVs=@n48.UVs raw, Tex=@n49, Level=@n50, Apply View MipBias) {mipValueMode=TMVM_MipLevel}.R)) {name=DistortionTexture_2, nodeColor=<rgba 0 0 0 1>}
  NamedRerouteDeclaration(None=Multiply(A=@n51, B=@n51) {constB=2}) {name=DistortionMap_2, nodeColor=<rgba 0 0 0 1>}
  n52 = NamedRerouteUsage()
  n53 = MaterialFunctionCall(Input UVs=@n52, Translate=AppendVector(A=Constant() {r=0.55}, B=Constant() {r=0.3641})) {materialFunction=&MF_UI_Translate}
  n54 = Constant2Vector() {g=0.5, r=1.4}
  n55 = Saturate(None=MaterialFunctionCall(UV=MaterialFunctionCall(UVs=@n53.UVs raw, Texture Scale=@n54) {materialFunction=&ScaleUVsByCenter}.UVs) {materialFunction=&MF_UI_RadialGradient}.Smooth)
  n56 = Constant() {r=0.515}
  n57 = Constant() {r=2.6}
  n58 = TextureSample!(
   UVs=MaterialFunctionCall(Input UVs=@n52, Translate=AppendVector(A=@n56, B=@n57)) {materialFunction=&MF_UI_Translate}.UVs raw
   Tex=TextureObject() {texture=&T_UI_HonMoon_Repeatable_BlurMask}
   Apply View MipBias
  ) {samplerType=SAMPLERTYPE_Alpha}
  NamedRerouteDeclaration(None=Multiply(A=LinearInterpolate(B=@n55, Alpha=@n58.R) {constB=0}, B=NamedRerouteUsage())) {name=Line Glow Mask, nodeColor=<rgba 0 0 0 1>}
  NamedRerouteDeclaration(None=@n23) {name=Honmoon Mask, nodeColor=<rgba 0 0 0 1>}
  n59 = NamedRerouteUsage()
  n60 = AppendVector(A=Constant() {r=0.15}, B=Constant() {r=0.15})
  n61 = NamedRerouteUsage()
  n62 = MaterialFunctionCall(
   UVs=MaterialFunctionCall(ScreenUV=MaterialFunctionCall(Input UVs=@n59, Translate=@n60) {materialFunction=&MF_UI_Translate}.UVs raw) {materialFunction=&MF_UI_ScreenSpaceUVScaling}.Scaled to Fit
   Size=LinearInterpolate(A=Constant(), B=Constant() {r=5}, Alpha=Power(Base=@n61) {constExponent=0.4})
   OutlineThickness=Constant() {r=0.12}
  ) {materialFunction=&MF_UI_Circle}
  n63 = LinearInterpolate(A=Constant() {r=0.08}) {constA=0.025, constB=0}
  n64 = OneMinus(None=MaterialFunctionCall!(In=Subtract(A=@n62.SDFOutline, B=@n63) {constB=0.005}, Ramp Width=Constant() {r=0.5}) {materialFunction=/Engine/Functions/Engine_MaterialFunctions01/Math/SmoothCeil}.Result)
  n65 = NamedRerouteUsage()
  n66 = Multiply(A=TextureSample(
    UVs=Panner(Coordinate=@n65) {speedX=-0.04, speedY=0.226}
    Tex=TextureObject() {texture=&T_UI_HonMoon_Repeatable_BlurMask}
   ) {samplerType=SAMPLERTYPE_Alpha}.R) {constB=2}
  n67 = Multiply(A=Power(Base=@n66), B=@n66)
  n68 = Multiply(A=@n66) {constB=0.9}
  NamedRerouteDeclaration(None=Multiply(A=Saturate(None=@n64), B=Add(A=@n67, B=@n68))) {name=Inner Ring Mask, nodeColor=<rgba 0 0 0 1>}
  n69 = NamedRerouteUsage()
  n70 = TextureSample(
   UVs=MaterialFunctionCall(
    UVs=MaterialFunctionCall(ScreenUV=@n69) {materialFunction=&MF_UI_ScreenSpaceUVScaling}.Scaled to Fill
    Texture Scale=Constant2Vector() {g=0.5, r=1}
   ) {materialFunction=&ScaleUVsByCenter}.UVs
   Tex=TextureObject() {samplerType=SAMPLERTYPE_Alpha, texture=/DiscoveryBrowser/Art/Textures/T_UI_Refrigerator_Background_flare}
  ) {samplerType=SAMPLERTYPE_Alpha}
  n71 = TextureSample(
   UVs=Panner(Coordinate=NamedRerouteUsage()) {speedX=-0.04, speedY=0.226}
   Tex=TextureObject() {texture=&T_UI_HonMoon_Repeatable_BlurMask}
  ) {samplerType=SAMPLERTYPE_Alpha}
  NamedRerouteDeclaration(None=Multiply(A=Multiply(A=@n70.R) {constB=0.75}, B=Multiply(A=@n71.R) {constB=2.1})) {name=flare mask, nodeColor=<rgba 0 0 0 1>}
  NamedRerouteDeclaration(None=LinearInterpolate(B=NamedRerouteUsage(), Alpha=NamedRerouteUsage())) {name=Honmoon Pulse Mask, nodeColor=<rgba 0 0 0 1>}
  n72 = Multiply(A=Time()) {constB=0.1}
  NamedRerouteDeclaration(None=Frac(None=Add(A=@n72) {constB=0.5})) {name=Animation, nodeColor=<rgba 0 0 0 1>}
  n73 = Constant() {r=-1.2}
  n74 = Constant() {r=-0.06}
  n75 = MaterialFunctionCall(
   UVs=MaterialFunctionCall(Input UVs=NamedRerouteUsage(), Translate=AppendVector(A=@n73, B=@n74)) {materialFunction=&MF_UI_Translate}.UVs raw
   Texture Scale=AppendVector(A=Constant() {r=1.6}, B=Constant() {r=1})
  ) {materialFunction=&ScaleUVsByCenter}
  n76 = Multiply(A=@n75.UVs, B=NamedRerouteUsage()) {constB=2}
  n77 = Constant() {r=0.002}
  n78 = Constant() {r=2}
  n79 = ScalarParameter() {defaultValue=0.2, group=Distortion, parameterName=Distortion_MotionSpeed}
  n80 = MaterialFunctionCall(
   UVs=MaterialFunctionCall(
    UVs=Add(A=@n75.UVs, B=@n76)
    Amplitude X=@n77
    Amplitude Y=@n77
    Frequency X=@n78
    Frequency Y=@n78
    Phase=Constant()
    Motion=MaterialFunctionCall(Speed=@n79) {materialFunction=&MF_UI_LinearTime}.0-1 Loop
   ) {materialFunction=&MF_UI_WaveWarp}.Warped UVs
   "Rotation Angle (0-1)"=Constant() {r=-0.07}
  ) {materialFunction=&CustomRotator}
  n81 = Constant2Vector() {r=0.03}
  n82 = Multiply(A=Clamp(None=Add(A=@n80.Rotated Values, B=@n81)) {maxDefault=0.9}, B=Constant2Vector() {g=1, r=24})
  n83 = NamedRerouteDeclaration(None=MaterialFunctionCall(UVs=Frac(None=@n82), PointA=Constant2Vector() {g=-2}, PointB=Constant2Vector() {g=2}) {materialFunction=&MF_UI_LineSegment}.Fill) {name=LinesMask_2, nodeColor=<rgba 0 0 0 1>}
  NamedRerouteDeclaration(None=@n83) {name=Honmoon Mask 2, nodeColor=<rgba 0 0 0 1>}
  NamedRerouteDeclaration(None=@n47.ScaleY) {name=Scale Y, nodeColor=<rgba 0 0 0 1>}
  n84 = NamedRerouteUsage()
  n85 = Multiply(A=MaterialFunctionCall(UV Channel=NamedRerouteUsage()) {materialFunction=&LinearGradient}.UGradient) {constB=0.6}
  NamedRerouteDeclaration(None=LinearInterpolate(B=Multiply(A=@n84, B=@n85), Alpha=NamedRerouteUsage())) {name=VignetteHonmoon mask, nodeColor=<rgba 0 0 0 1>}
  n86 = Constant() {r=0.2}
  n87 = NamedRerouteUsage()
  n88 = Constant2Vector() {g=0.17, r=-0.6954}
  n89 = MaterialFunctionCall(UV=MaterialFunctionCall(
    UVs=MaterialFunctionCall(Input UVs=@n87, Translate=@n88) {materialFunction=&MF_UI_Translate}.UVs raw
    Texture Scale=Constant2Vector() {g=1, r=1}
   ) {materialFunction=&ScaleUVsByCenter}.UVs) {materialFunction=&MF_UI_RadialGradient}
  NamedRerouteDeclaration(None=LinearInterpolate(B=Multiply(A=@n83, B=@n86), Alpha=OneMinus(None=@n89.Smooth))) {name="Honmoon Mask_2 (dim)", nodeColor=<rgba 0 0 0 1>}
  n90 = MaterialFunctionCall(UVs=Add(A=NamedRerouteUsage(), B=Constant2Vector() {g=0.25, r=0.55}), Texture Scale=Constant2Vector() {g=1.5, r=1.5}) {materialFunction=&ScaleUVsByCenter}
  n91 = Multiply(A=MaterialFunctionCall("Value (0-1)"=MaterialFunctionCall(UV=@n90.UVs) {materialFunction=&MF_UI_RadialGradient}.Smooth) {materialFunction=&MF_UI_EaseCurves}.EaseInOutSine) {constB=0.7}
  n92 = NamedRerouteUsage()
  n93 = Constant2Vector() {g=-0.65, r=0.45}
  n94 = MaterialFunctionCall(UV=MaterialFunctionCall(UVs=Add(A=@n92, B=@n93), Texture Scale=Constant2Vector() {g=1.6, r=1.6}) {materialFunction=&ScaleUVsByCenter}.UVs) {materialFunction=&MF_UI_RadialGradient}
  n95 = NamedRerouteDeclaration(None=Multiply(A=MaterialFunctionCall("Value (0-1)"=@n94.Smooth) {materialFunction=&MF_UI_EaseCurves}.EaseInOutSine) {constB=0.7}) {name=glow2, nodeColor=<rgba 0 0 0 1>}
  n96 = NamedRerouteUsage()
  n97 = Constant2Vector() {g=-0.58, r=-0.45}
  n98 = MaterialFunctionCall(UV=MaterialFunctionCall(UVs=Add(A=@n96, B=@n97), Texture Scale=Constant2Vector() {g=2, r=1.6}) {materialFunction=&ScaleUVsByCenter}.UVs) {materialFunction=&MF_UI_RadialGradient}
  n99 = NamedRerouteDeclaration(None=Multiply(A=MaterialFunctionCall("Value (0-1)"=@n98.Smooth) {materialFunction=&MF_UI_EaseCurves}.EaseInOutSine) {constB=0.8}) {name=glow 1, nodeColor=<rgba 0 0 0 1>}
  NamedRerouteDeclaration(None=LinearInterpolate(A=NamedRerouteDeclaration(None=@n91) {name=glow3, nodeColor=<rgba 0 0 0 1>}, Alpha=LinearInterpolate(A=@n95, Alpha=@n99))) {name=Glow Masks, nodeColor=<rgba 0 0 0 1>}
 ]
}
