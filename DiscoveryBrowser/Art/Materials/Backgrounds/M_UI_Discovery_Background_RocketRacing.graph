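# Translucent UI material that draws the animated Rocket Racing background for Discovery browser IP pages, output through MF_UI_Discovery_IP_Page.
# A purple vertical gradient blends with a screen-scaled radial gradient to form the base.
# A flag pattern texture, fitted and mirrored onto left and right screen halves with edge clipping, is overlay-blended faintly on top.
# Time-scrolled, rotated sine offsets ripple the flag UVs, eased to zero toward the edges.
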Material /DiscoveryBrowser/Art/Materials/Backgrounds/M_UI_Discovery_Background_RocketRacing
graph {
 properties {
  bAutomaticallySetUsageInEditor=false
  bUsedWithStaticMesh=false
  blendMode=BLEND_Translucent
  materialDomain=MD_UI
  usedShadingModels=Default Lit
 }
 nodes [
  MaterialFunctionCall!(Background Color=NamedRerouteUsage()) {materialFunction=/DiscoveryBrowser/Art/MaterialFunctions/MF_UI_Discovery_IP_Page}
  n1 = ScreenPosition()
  n2 = MaterialFunctionCall!(ScreenUV=Add(A=@n1.ViewportUV, B=Constant2Vector() {g=0.5, r=0.5}), TextureSize, ScalePivotPoint) {materialFunction=/FortUILibrary/MaterialFunctions/Basics/MF_UI_ScreenSpaceUVScaling}
  n3 = Constant() {r=6}
  n4 = LinearInterpolate(
   A=LinearInterpolate(
    A=Constant3Vector() {constant=<rgba 0.0482 0.0065 0.521 1>}
    B=Constant3Vector() {constant=<rgba 0.0232 0.0052 0.0742 1>}
    Alpha=ComponentMask(None=@n1.ViewportUV) {g=true}
   )
   B=Constant3Vector() {constant=<rgba 0.0152 0.0027 0.0561 1>}
   Alpha=MaterialFunctionCall(UV=MaterialFunctionCall(UVs=@n2.ScaleY, Texture Scale=@n3) {materialFunction=&ScaleUVsByCenter}.UVs) {materialFunction=&MF_UI_RadialGradient}.Smooth
  )
  n5 = NamedRerouteUsage()
  NamedRerouteDeclaration(None=LinearInterpolate(
    A=@n4
    B=MaterialFunctionCall(Base=@n4, Blend=@n5) {materialFunction=&Blend_Overlay}.Result
    Alpha=Multiply(A=@n5) {constB=0.2}
   )) {name=Background, nodeColor=<rgba 0 0 0 1>}
  n6 = NamedRerouteUsage()
  n7 = TextureProperty(Tex=NamedRerouteDeclaration(None=TextureObject() {samplerType=SAMPLERTYPE_Alpha, texture=/DiscoveryBrowser/Art/Textures/T_UI_Discovery_RocketRacing_Pattern}) {name=Texture Object, nodeColor=<rgba 0 0 0 1>})
  n8 = Add(
   A=MaterialFunctionCall(ScreenUV=@n6, TextureSize=@n7, ScalePivotPoint=Constant2Vector() {g=1}) {materialFunction=&MF_UI_ScreenSpaceUVScaling}.Scaled to Fit
   B=Constant2Vector() {r=0.5}
  ) {desc=Left Side}
  n9 = Add(
   A=MaterialFunctionCall(ScreenUV=OneMinus(None=@n6), TextureSize=@n7, ScalePivotPoint=Constant2Vector() {g=1}) {materialFunction=&MF_UI_ScreenSpaceUVScaling}.Scaled to Fit
   B=Constant2Vector() {r=0.575}
  ) {desc=Right Side}
  n10 = Round(None=ComponentMask(None=NamedRerouteUsage()) {r=true})
  n11 = TextureSample(
   UVs=NamedRerouteDeclaration(None=LinearInterpolate(A=@n8, B=@n9, Alpha=@n10)) {name=Mirror Flag UV, nodeColor=<rgba 0 0 0 1>}
   Tex=NamedRerouteUsage()
  ) {
   samplerSource=SSM_Clamp_WorldGroupSettings
   samplerType=SAMPLERTYPE_Alpha
   texture=&T_UI_Discovery_RocketRacing_Pattern
  }
  n12 = NamedRerouteUsage()
  NamedRerouteDeclaration(None=Multiply(
    A=@n11.R
    B=MaterialFunctionCall!(UVs=@n12, Alpha=@n11.A) {materialFunction=/FortUILibrary/MaterialFunctions/Basics/MF_UI_ClipEdgePixel}.Result
   )) {name=Pattern, nodeColor=<rgba 0 0 0 1>}
  n13 = NamedRerouteUsage()
  n14 = ScreenPosition()
  n15 = Time()
  n16 = Add(
   A=Rotator(Coordinate=NamedRerouteDeclaration(None=@n14.ViewportUV) {name=Base UV, nodeColor=<rgba 0 0 0 1>}, Time=Constant() {r=0.075}) {speed=6.2832}
   B=AppendVector(A=Constant(), B=Multiply(A=@n15) {constB=0.25}) {desc=Flag Movement Speed}
  ) {constB=0}
  n17 = LinearInterpolate(A=Constant2Vector(), B=Constant2Vector() {g=0.05}, Alpha=Sine(None=ConstantBiasScale(None=@n16) {bias=0})) {constB=0.01}
  n18 = ComponentMask(None=NamedRerouteUsage()) {g=true}
  n19 = OneMinus(None=@n18)
  n20 = MaterialFunctionCall("Value (0-1)"=Multiply(A=Min(A=@n18, B=@n19)) {constB=2}) {materialFunction=&MF_UI_EaseCurves}
  NamedRerouteDeclaration(None=LinearInterpolate(A=@n13, B=Add(A=@n13, B=@n17), Alpha=Saturate(None=@n20.EaseOutQuint))) {name=Flag UV, nodeColor=<rgba 0 0 0 1>}
 ]
}
