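# Translucent UI material for the Discovery browser's SolidWave background: an animated tiled pattern over blue gradients.
# Screen-space UVs form a rotated grid whose random per-cell values pick texture channels for pattern mask and colour.
# Time-driven sine and ease curves animate rotation of the pattern samples.
# Gradient, pattern, radial glow and vignette layers blend into the background colour passed to MF_UI_Discovery_IP_Page.
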Material /DiscoveryBrowser/Art/Materials/Backgrounds/M_UI_Discovery_Background_SolidWave
graph {
 properties {
  bAutomaticallySetUsageInEditor=false
  bUsedWithStaticMesh=false
  blendMode=BLEND_Translucent
  materialDomain=MD_UI
  usedShadingModels=Default Lit
 }
 nodes [
  Constant() {r=0.15}
  MaterialFunctionCall() {materialFunction=&GetUserInterfaceUV}
  n1 = Constant3Vector() {constant=<rgba 0.0343 0.2423 0.3968 1>}
  n2 = Constant3Vector() {constant=<rgba 0.1022 0.4397 0.5776 1>}
  n3 = Multiply(A=@n2, B=NamedRerouteUsage())
  n4 = LinearInterpolate(A=LinearInterpolate(A=NamedRerouteUsage(), B=@n2, Alpha=@n3), B=@n3, Alpha=Multiply(A=NamedRerouteUsage()) {constB=0.4})
  n5 = Constant3Vector() {constant=<rgba 0.5647 0.6105 0.5711 1>}
  n6 = Add(A=ScreenPosition().ViewportUV, B=AppendVector(A=Constant() {r=-0.26}, B=Constant() {r=0.15}))
  n7 = Constant() {r=5.2}
  n8 = Power(
   Base=MaterialFunctionCall(UV=MaterialFunctionCall(UVs=@n6, Texture Scale=@n7) {materialFunction=&ScaleUVsByCenter}.UVs) {materialFunction=&MF_UI_RadialGradient}.Smooth
   Exp=Constant() {r=5}
  )
  n9 = LinearInterpolate(A=@n4, B=Saturate(None=Add(A=@n4, B=@n5)), Alpha=Multiply(A=@n8) {constB=0.2})
  n10 = NamedRerouteUsage()
  MaterialFunctionCall!(Background Color=LinearInterpolate(A=LinearInterpolate(A=@n1, B=@n9, Alpha=@n8), B=Constant3Vector() {constant=<rgba 0 0 0 1>}, Alpha=OneMinus(None=@n10))) {materialFunction=/DiscoveryBrowser/Art/MaterialFunctions/MF_UI_Discovery_IP_Page}
  n11 = NamedRerouteUsage()
  NamedRerouteDeclaration(None=LinearInterpolate(A=Constant3Vector() {constant=<rgba 1 0.0999 0.3968 1>}, Alpha=OneMinus(None=@n11)) {constB=0}) {name=pink vignette, nodeColor=<rgba 0 0 0 1>}
  n12 = ScreenPosition()
  n13 = Constant() {r=-0.5}
  n14 = Constant()
  n15 = Constant()
  n16 = MaterialFunctionCall(
   UVs=Add(
    A=MaterialFunctionCall(UVs=@n12.ViewportUV, Rotation Center=@n13) {materialFunction=&CustomRotator}.Rotated Values
    B=AppendVector(A=@n14, B=@n15)
   )
   Texture Scale=Constant() {r=2}
  ) {materialFunction=&ScaleUVsByCenter}
  NamedRerouteDeclaration(None=Power(
    Base=MaterialFunctionCall(UV=@n16.UVs) {materialFunction=&MF_UI_RadialGradient}.Smooth
    Exp=Constant() {r=0.9}
   )) {name=Vignette, nodeColor=<rgba 0 0 0 1>}
  n17 = NamedRerouteUsage()
  n18 = Add(A=Multiply(A=Time(), B=Constant() {desc=Speed, r=0.3}), B=NamedRerouteUsage())
  n19 = MaterialFunctionCall("Value (0-1)"=ConstantBiasScale(None=Sine(None=@n18))) {materialFunction=&MF_UI_EaseCurves}
  n20 = Cosine(None=NamedRerouteUsage())
  n21 = OneMinus(None=Floor(None=Add(A=@n20) {constB=0.1}))
  n22 = Rotator(
   Coordinate=MaterialFunctionCall(UVs=NamedRerouteUsage(), Texture Scale=LinearInterpolate(Alpha=@n17) {constA=0.5, constB=0.9}) {materialFunction=&ScaleUVsByCenter}.UVs
   Time=LinearInterpolate(
    A=LinearInterpolate(Alpha=@n19.EaseOutSine) {constA=-0.01, constB=-0.05}
    B=LinearInterpolate(Alpha=@n19.EaseInOutSine) {constA=-0.18, constB=-0.2}
    Alpha=Saturate(None=@n21)
   )
  ) {speed=6.2832}
  n23 = TextureSample(UVs=@n22) {samplerType=SAMPLERTYPE_LinearColor, texture=/DiscoveryBrowser/Art/Textures/T_UI_SolidWave_Background_Mask}
  n24 = MaterialFunctionCall(
   UVs=Multiply(A=MaterialFunctionCall!(ScreenUV, TextureSize, ScalePivotPoint) {materialFunction=/FortUILibrary/MaterialFunctions/Basics/MF_UI_ScreenSpaceUVScaling}.Scaled to Fit) {constB=6}
   "Rotation Angle (0-1)"=Constant() {r=0.1}
  ) {materialFunction=&CustomRotator}
  n25 = MaterialFunctionCall!(UV=Floor(None=@n24.Rotated Values), Seed=Constant() {r=4.4}) {materialFunction=&MF_UI_Random}
  n26 = Saturate(None=Multiply(A=NamedRerouteDeclaration(None=@n25.Result) {name=Random Cells, nodeColor=<rgba 0 1 0.4375 1>}, B=Constant() {r=1.5}) {constB=1.3})
  n27 = Floor(None=@n26)
  n28 = Min(A=OneMinus(None=@n27), B=Round(None=@n26))
  NamedRerouteDeclaration(None=LinearInterpolate(A=LinearInterpolate(A=@n23.G, B=@n23.G, Alpha=@n28), B=@n23.R, Alpha=@n27)) {name=Pattern Alpha, nodeColor=<rgba 0 0 0 1>}
  n29 = TextureSample(UVs=@n22) {samplerType=SAMPLERTYPE_LinearColor, texture=/DiscoveryBrowser/Art/Textures/T_UI_SolidWave_Background_BaseColour}
  NamedRerouteDeclaration(None=LinearInterpolate(A=LinearInterpolate(A=@n29.G, B=@n29.G, Alpha=@n28), B=@n29.R, Alpha=@n27)) {name=Pattern Colour, nodeColor=<rgba 0 0 0 1>}
  NamedRerouteDeclaration(None=Frac(None=@n24.Rotated Values)) {name=Pattern UV, nodeColor=<rgba 0.5234 1 0 1>}
  n30 = ScreenPosition()
  NamedRerouteDeclaration(None=LinearInterpolate(
    A=Constant3Vector() {constant=<rgba 0.1022 0.4397 0.5776 1>}
    B=Constant3Vector() {constant=<rgba 0.0356 0.2462 0.402 1>}
    Alpha=ComponentMask(None=@n30.ViewportUV) {g=true}
   )) {name=Gradient, nodeColor=<rgba 0 0 0 1>}
  n31 = ScreenPosition()
  n32 = Constant() {r=-0.58}
  n33 = ComponentMask(None=Subtract(A=MaterialFunctionCall(UVs=@n31.ViewportUV, "Rotation Angle (0-1)"=@n32) {materialFunction=&CustomRotator}.Rotated Values) {constB=0.6}) {g=true}
  NamedRerouteDeclaration(None=Saturate(None=Multiply(A=@n33) {constB=0.35})) {name=P_Gradient Mask, nodeColor=<rgba 0 0 0 1>}
 ]
}
